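; Gameplay alerts director component that shows and removes the UI scene for the current Battle Royale gameplay alert.
; From BeginPlay until EndPlay it listens to the stateful CurrentGameplayAlertChanged event, resolving the alert tag to a scene through the GameplayAlertsScenesInfo Data Registry.
; It adds scenes one frame later so a scene removed and re-added in one tick still reconstructs its widget.
; Clearing the event state removes the valid scene and broadcasts RemoveGameplayAlertRequested.
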(event EventBeginPlay
  ;[1 1 1] Begin play - starts to listen to event
  (bind (_asynctaskproxy _payload) (GameplayEventRouter|StartListeningToStatefulEvent (GameplayEventRouter|GetGlobalEventRouter) "/Script/BattleRoyaleGameplayAlerts.FortEvent_CurrentGameplayAlertChanged")
    (:then
      (Variables|Default|SetAsync%20Task%20Ref _asynctaskproxy))
    (:OnEventReceived
      _)
    (:OnSavedEventStateExists
      _)
    (:OnEventStateCleared
      (CallFunction|RemoveGameplayAlertScene)))
  (bind _output_get (Variables|Default|SetCurrentEventPayload _payload))
  (bind _output_get_gameplayalerttag (Utilities|Struct|BreakFortEventCurrentGameplayAlertChanged _output_get))
  (Variables|Default|SetCurrentAlertUIScene (CallFunction|GetUISceneforAlert _output_get_gameplayalerttag))
  (Utilities|FlowControl|DelayUntilNextFrame) ; Wait for the next Engine Tick before\nadding the scene to ensure that\n the last alert's Dynamic UI Scene \nhas been removed before adding \nthe new scene.\n\nFixes an issue where if the same scene\nis removed and added in the same Tick\nthe widget will not be reconstructed and\nits Event Construct and \nEvent OnInitialized are never called, \nwhich might make the\nwidget stay on screen permanently.
  (CallFunction|AddGameplayAlertScene))

(event Custom|AddGameplayAlertScene
  (bind _currentalertuiscene (Variables|Default|GetCurrentAlertUIScene))
  ;[1 1 1] Add gameplay alert scene
  (Utilities|IsValid _currentalertuiscene
    (:Is%20Valid
      (CallFunction|AddScene _currentalertuiscene))))

(event Custom|RemoveGameplayAlertScene
  (bind _currentalertuiscene (Variables|Default|GetCurrentAlertUIScene))
  (bind _currenteventpayload_gameplayalerttag (Utilities|Struct|BreakFortEventCurrentGameplayAlertChanged (Variables|GetCurrentEventPayload_GameplayAlertTag)))
  ;[1 1 1] Remove gameplay alert scene
  (Utilities|IsValid _currentalertuiscene
    (:Is%20Valid
      (CallFunction|RemoveScene _currentalertuiscene)
      (Variables|SetRemoveEvent_GameplayAlertTag (Utilities|Struct|MakeFortEventRemoveGameplayAlertRequested _currenteventpayload_gameplayalerttag))
      (GameplayEventRouter|BroadcastEvent (GameplayEventRouter|GetGlobalEventRouter) (Variables|Default|GetRemoveEvent) 0))))

(event EventEndPlay
  (params (EndPlayReason :type EEndPlayReason%20Enum :default "Destroyed"))
  ;[1 1 1] End play
  (GameplayEventRouter|StopListeningtoEvent (Variables|Default|GetEventAsyncTaskRef)))

(fn GetUISceneForAlert
  (params (GameplayAlertTag :type Gameplay%20Tag%20Structure))
  (bind (_dataregistryoutitem_alertgameplaytag_7_45f91c4640c38b57bd8e42b36fe3ff59 _dataregistryoutitem_alertuiscene_5_846ed33740274597bb10ab9a6a48bff2) (Utilities|Struct|BreakGameplayAlertSceneInfo (Variables|GetDataRegistryOutItem_AlertGameplayTag_7_45F91C4640C38B57BD8E42B36FE3FF59)))
  (bind _returnvalue (DataRegistry|GetDataRegistryItem (Utilities|Struct|MakeDataRegistryId "GameplayAlertsScenesInfo" (GameplayTags|GetTagName GameplayAlertTag)) (LocalVariables|GetDataRegistryOutItem)))
  (if _returnvalue
    (return _dataregistryoutitem_alertuiscene_5_846ed33740274597bb10ab9a6a48bff2)
    (else
      (return 0))))
